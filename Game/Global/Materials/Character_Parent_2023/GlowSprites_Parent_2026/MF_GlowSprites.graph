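# Material function for camera-facing glow sprites, outputting material attributes plus baked vertex-color random seeds and sprite IDs.
# World position offset expands sprites along camera axes, with camera offset, unless angle fade is enabled.
# Vertex alpha times sprite count gives an ID that can select from 16 colors and sizes.
# A sphere gradient, mask or custom input shapes the glow, alongside depth, camera and angle fades, time-of-day scaling and an ID debug view.
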MaterialFunction /Game/Global/Materials/Character_Parent_2023/GlowSprites_Parent_2026/MF_GlowSprites
graph {expressions {
  MaterialExpressionAdd_0 {
   class=MaterialExpressionAdd
   inputs [
    A {from=MaterialExpressionAdd_8}
    B {from=MaterialExpressionMaterialFunctionCall_1, output=Clamped Camera Offset}
   ]
  }
  MaterialExpressionAdd_7 {
   class=MaterialExpressionAdd
   inputs [
    A {from=MaterialExpressionFloor_1}
    B {from=""}
   ]
   properties {constB=0.5}
  }
  MaterialExpressionAdd_8 {
   class=MaterialExpressionAdd
   inputs [
    A {from=MaterialExpressionMultiply_18}
    B {from=MaterialExpressionMultiply_19}
   ]
  }
  MaterialExpressionChannelMaskParameter_0 {
   class=MaterialExpressionChannelMaskParameter
   inputs [None {from=MaterialExpressionNamedRerouteUsage_34, output=""}]
   properties {group=GlowSprites 07 Randomness, maskChannel=Green, parameterName=RandomSize_Seed, defaultValue=<rgba 0 1 0 0>}
  }
  MaterialExpressionClamp_0 {
   class=MaterialExpressionClamp
   inputs [
    None {from=MaterialExpressionScalarParameter_13, output=""}
    Min {from="", output=""}
    Max {from="", output=""}
   ]
   properties {clampMode=CMODE_ClampMin, maxDefault=100, minDefault=1}
  }
  MaterialExpressionComponentMask_0 {class=MaterialExpressionComponentMask, inputs [None {from=MaterialExpressionVertexColor_0}], properties {b=true, g=true, r=true}}
  MaterialExpressionConstant3Vector_0 {class=MaterialExpressionConstant3Vector, properties {constant=<rgba 1 1 1 0>}}
  MaterialExpressionConstant3Vector_22 {class=MaterialExpressionConstant3Vector, properties {constant=<rgba 0 -1 0 0>}}
  MaterialExpressionConstant3Vector_23 {class=MaterialExpressionConstant3Vector, properties {constant=<rgba 1 0 0 0>}}
  MaterialExpressionConstant_0 {class=MaterialExpressionConstant}
  MaterialExpressionConstant_2 {class=MaterialExpressionConstant}
  MaterialExpressionConstant_4 {class=MaterialExpressionConstant, properties {r=2}}
  MaterialExpressionConstant_5 {class=MaterialExpressionConstant, properties {r=0.5}}
  MaterialExpressionConstant_7 {class=MaterialExpressionConstant, properties {r=1}}
  MaterialExpressionConstant_8 {class=MaterialExpressionConstant, properties {r=1}}
  MaterialExpressionConvert_2 {
   class=MaterialExpressionConvert
   inputs [
    None {from=MaterialExpressionDivide_3}
    None {from=""}
   ]
   properties {
    convertInputs [
     {
      type=Scalar
      expressionInput {
       expression=&self:MaterialExpressionDivide_3
       inputName=None
       mask=0
       maskA=0
       maskB=0
       maskG=0
       maskR=0
       outputIndex=0
      }
     }
     {type=Scalar, expressionInput {expression=null, inputName=None, mask=0, maskA=0, maskB=0, maskG=0, maskR=0, outputIndex=0}}
    ]
    convertOutputs [{type=Vector2}]
   }
  }
  MaterialExpressionDepthFade_1 {
   class=MaterialExpressionDepthFade
   inputs [
    Opacity {from=MaterialExpressionMaterialFunctionCall_3, output=Result}
    FadeDistance {from=MaterialExpressionMultiply_22, output=""}
   ]
   properties {}
  }
  MaterialExpressionDivide_1 {
   class=MaterialExpressionDivide
   inputs [
    A {from=MaterialExpressionObjectRadius_0}
    B {from=MaterialExpressionScalarParameter_4}
   ]
  }
  MaterialExpressionDivide_3 {
   class=MaterialExpressionDivide
   inputs [
    A {from=MaterialExpressionAdd_7}
    B {from=MaterialExpressionNamedRerouteUsage_13}
   ]
  }
  MaterialExpressionFloor_1 {class=MaterialExpressionFloor, inputs [None {from=MaterialExpressionNamedRerouteUsage_17, output=""}], properties {}}
  MaterialExpressionFunctionInput_0 {
   class=MaterialExpressionFunctionInput
   inputs [Preview {from=MaterialExpressionNamedRerouteUsage_5, output=""}]
   properties {bUsePreviewValueAsDefault=true, inputName=CustomGlowGraph, inputType=FunctionInput_Scalar, sortPriority=1}
  }
  MaterialExpressionFunctionInput_1 {
   class=MaterialExpressionFunctionInput
   inputs [Preview {from=MaterialExpressionConstant3Vector_0, output=""}]
   properties {bUsePreviewValueAsDefault=true, inputName=CustomGlowColor, sortPriority=2, previewValue=<1 1 1 1>}
  }
  MaterialExpressionFunctionInput_2 {
   class=MaterialExpressionFunctionInput
   inputs [Preview {from=MaterialExpressionStaticBool_0, output=""}]
   properties {bUsePreviewValueAsDefault=true, inputName=UseCustomGlowGraph, inputType=FunctionInput_StaticBool, sortPriority=1}
  }
  MaterialExpressionFunctionOutput_0 {
   class=MaterialExpressionFunctionOutput
   inputs [None {from=MaterialExpressionSetMaterialAttributes_0, output=""}]
   properties {outputName=MA}
  }
  MaterialExpressionFunctionOutput_1 {
   class=MaterialExpressionFunctionOutput
   inputs [None {from=MaterialExpressionNamedRerouteUsage_12, output=""}]
   properties {description="Read baked vertex color:\r\nR - Random Seed\r\nG - Random Seed\r\nB - Random Seed", outputName=VertexColor RGB, sortPriority=1}
  }
  MaterialExpressionFunctionOutput_2 {
   class=MaterialExpressionFunctionOutput
   inputs [None {from=MaterialExpressionNamedRerouteUsage_14, output=""}]
   properties {description="Read baked vertex color:\r\nA - Sprite ID", outputName=Sprites ID, sortPriority=2}
  }
  MaterialExpressionLinearInterpolate_0 {
   class=MaterialExpressionLinearInterpolate
   inputs [
    A {from=MaterialExpressionScalarParameter_1}
    B {from=MaterialExpressionScalarParameter_3}
    Alpha {from=MaterialExpressionChannelMaskParameter_0, output=R}
   ]
  }
  MaterialExpressionLinearInterpolate_19 {
   class=MaterialExpressionLinearInterpolate
   inputs [
    A {from=MaterialExpressionReroute_1}
    B {from=MaterialExpressionMultiply_58}
    Alpha {from=MaterialExpressionScalarParameter_19}
   ]
  }
  MaterialExpressionLinearInterpolate_5 {
   class=MaterialExpressionLinearInterpolate
   inputs [
    A {from=MaterialExpressionNamedRerouteUsage_8}
    B {from=MaterialExpressionMultiply_4}
    Alpha {from=MaterialExpressionScalarParameter_20}
   ]
  }
  MaterialExpressionMaterialFunctionCall_1 {
   class=MaterialExpressionMaterialFunctionCall
   inputs [
    Offset Amount {from=MaterialExpressionScalarParameter_0}
    Clamp Padding {from=""}
    World Position {from=""}
   ]
   properties {materialFunction=/Engine/Functions/Engine_MaterialFunctions02/WorldPositionOffset/CameraOffset}
  }
  MaterialExpressionMaterialFunctionCall_2 {
   class=MaterialExpressionMaterialFunctionCall
   inputs [
    Number {from=MaterialExpressionRound_3}
    MaximumNumberOfDigits {from=MaterialExpressionConstant_4}
    UVs {from=MaterialExpressionSubtract_0}
    DebugTextLocation RG_UpperRight BA_LowerLeft {from=""}
   ]
   properties {materialFunction=&DebugScalarValues}
  }
  MaterialExpressionMaterialFunctionCall_3 {
   class=MaterialExpressionMaterialFunctionCall
   inputs [Distance {from=MaterialExpressionScalarParameter_14}]
   properties {materialFunction=&Near_Camera_Fade}
  }
  MaterialExpressionMaterialFunctionCall_4 {
   class=MaterialExpressionMaterialFunctionCall
   inputs []
   properties {materialFunction=&ObjectScale}
  }
  MaterialExpressionMaterialFunctionCall_5 {
   class=MaterialExpressionMaterialFunctionCall
   inputs [
    Value_Morning {from=MaterialExpressionScalarParameter_15}
    Value_Day {from=MaterialExpressionScalarParameter_16}
    Value_Evening {from=MaterialExpressionScalarParameter_17}
    Value_Night {from=MaterialExpressionScalarParameter_18}
   ]
   properties {materialFunction=&Scalar_Time_of_Day}
  }
  MaterialExpressionMaterialFunctionCall_6 {
   class=MaterialExpressionMaterialFunctionCall
   inputs [
    Fade Length {from=MaterialExpressionScalarParameter_56}
    Fade Offset {from=MaterialExpressionScalarParameter_57}
    For Vertex Shader {from=""}
   ]
   properties {materialFunction=&CameraDepthFade}
  }
  MaterialExpressionMaterialFunctionCall_66 {
   class=MaterialExpressionMaterialFunctionCall
   inputs [Float2 {from=MaterialExpressionSubtract_5}]
   properties {materialFunction=&BreakOutFloat2Components}
  }
  MaterialExpressionMaterialFunctionCall_67 {
   class=MaterialExpressionMaterialFunctionCall
   inputs [Float2 {from=MaterialExpressionLinearInterpolate_5}]
   properties {materialFunction=&BreakOutFloat2Components}
  }
  MaterialExpressionMaterialFunctionCall_7 {
   class=MaterialExpressionMaterialFunctionCall
   inputs ["FadeSensitivity (optional)" {from=MaterialExpressionScalarParameter_21}]
   properties {materialFunction=&FadeBasedOnViewAngle}
  }
  MaterialExpressionMaterialFunctionCall_9 {
   class=MaterialExpressionMaterialFunctionCall
   inputs [Emissive {from=MaterialExpressionMultiply_1}]
   properties {materialFunction=&MF_FX_EmissiveAdjustments}
  }
  MaterialExpressionMultiply_0 {
   class=MaterialExpressionMultiply
   inputs [
    A {from=MaterialExpressionMultiply_26}
    B {from=MaterialExpressionNamedRerouteUsage_20}
   ]
  }
  MaterialExpressionMultiply_1 {
   class=MaterialExpressionMultiply
   inputs [
    A {from=MaterialExpressionNamedRerouteUsage_7}
    B {from=MaterialExpressionNamedRerouteUsage_1}
   ]
  }
  MaterialExpressionMultiply_15 {
   class=MaterialExpressionMultiply
   inputs [
    A {from=MaterialExpressionVertexColor_0, output=A}
    B {from=MaterialExpressionNamedRerouteDeclaration_9}
   ]
   properties {constB=16}
  }
  MaterialExpressionMultiply_16 {
   class=MaterialExpressionMultiply
   inputs [
    A {from=MaterialExpressionMaterialFunctionCall_66, output=R}
    B {from=MaterialExpressionTransform_9}
   ]
  }
  MaterialExpressionMultiply_17 {
   class=MaterialExpressionMultiply
   inputs [
    A {from=MaterialExpressionMaterialFunctionCall_66, output=G}
    B {from=MaterialExpressionTransform_10}
   ]
  }
  MaterialExpressionMultiply_18 {
   class=MaterialExpressionMultiply
   inputs [
    A {from=MaterialExpressionMultiply_16}
    B {from=MaterialExpressionMaterialFunctionCall_67, output=R}
   ]
  }
  MaterialExpressionMultiply_19 {
   class=MaterialExpressionMultiply
   inputs [
    A {from=MaterialExpressionMultiply_17}
    B {from=MaterialExpressionMaterialFunctionCall_67, output=G}
   ]
  }
  MaterialExpressionMultiply_20 {
   class=MaterialExpressionMultiply
   inputs [
    A {from=MaterialExpressionMultiply_0}
    B {from=MaterialExpressionNamedRerouteUsage_19}
   ]
  }
  MaterialExpressionMultiply_22 {
   class=MaterialExpressionMultiply
   inputs [
    A {from=MaterialExpressionScalarParameter_24}
    B {from=MaterialExpressionScalarParameter_23}
   ]
  }
  MaterialExpressionMultiply_25 {
   class=MaterialExpressionMultiply
   inputs [
    A {from=MaterialExpressionNamedRerouteUsage_8}
    B {from=MaterialExpressionMaterialFunctionCall_4, output=Scale XYZ}
   ]
  }
  MaterialExpressionMultiply_26 {
   class=MaterialExpressionMultiply
   inputs [
    A {from=MaterialExpressionStaticSwitch_0}
    B {from=MaterialExpressionScalarParameter_26}
   ]
  }
  MaterialExpressionMultiply_3 {
   class=MaterialExpressionMultiply
   inputs [
    A {from=MaterialExpressionNamedRerouteUsage_22}
    B {from=""}
   ]
   properties {constB=2}
  }
  MaterialExpressionMultiply_4 {
   class=MaterialExpressionMultiply
   inputs [
    A {from=MaterialExpressionNamedRerouteUsage_8}
    B {from=MaterialExpressionDivide_1}
   ]
  }
  MaterialExpressionMultiply_58 {
   class=MaterialExpressionMultiply
   inputs [
    A {from=MaterialExpressionReroute_1}
    B {from=MaterialExpressionMaterialFunctionCall_6, output=Result}
   ]
  }
  MaterialExpressionNamedRerouteDeclaration_0 {
   class=MaterialExpressionNamedRerouteDeclaration
   inputs [None {from=MaterialExpressionRayTracingQualitySwitch_0}]
   properties {name=Out Opacity, nodeColor=<rgba 0 0.3672 1 1>}
  }
  MaterialExpressionNamedRerouteDeclaration_1 {
   class=MaterialExpressionNamedRerouteDeclaration
   inputs [None {from=MaterialExpressionStaticSwitchParameter_5}]
   properties {name=Out WPO, nodeColor=<rgba 0 1 0.5078 1>}
  }
  MaterialExpressionNamedRerouteDeclaration_10 {
   class=MaterialExpressionNamedRerouteDeclaration
   inputs [None {from=MaterialExpressionComponentMask_0}]
   properties {name=VC_RandomSeed, nodeColor=<rgba 1 1 1 1>}
  }
  MaterialExpressionNamedRerouteDeclaration_11 {
   class=MaterialExpressionNamedRerouteDeclaration
   inputs [None {from=MaterialExpressionMultiply_15}]
   properties {name=Sprite ID, nodeColor=<rgba 1 0.8438 0 1>}
  }
  MaterialExpressionNamedRerouteDeclaration_12 {
   class=MaterialExpressionNamedRerouteDeclaration
   inputs [None {from=MaterialExpressionTextureCoordinate_0}]
   properties {name=GlowSpriteUV, nodeColor=<rgba 1 0 0.8672 1>}
  }
  MaterialExpressionNamedRerouteDeclaration_13 {
   class=MaterialExpressionNamedRerouteDeclaration
   inputs [None {from=MaterialExpressionMaterialFunctionCall_5, output=Result}]
   properties {name=TOD_Factor, nodeColor=<rgba 1 0.5156 0 1>}
  }
  MaterialExpressionNamedRerouteDeclaration_14 {
   class=MaterialExpressionNamedRerouteDeclaration
   inputs [None {from=MaterialExpressionPower_0}]
   properties {name=DefaultGlowGraph, nodeColor=<rgba 1 0.75 0 1>}
  }
  MaterialExpressionNamedRerouteDeclaration_15 {
   class=MaterialExpressionNamedRerouteDeclaration
   inputs [None {from=MaterialExpressionFunctionInput_2}]
   properties {name=UseCustomGlowGraph, nodeColor=<rgba 1 0 0 1>}
  }
  MaterialExpressionNamedRerouteDeclaration_16 {
   class=MaterialExpressionNamedRerouteDeclaration
   inputs [None {from=MaterialExpressionStaticSwitchParameter_6}]
   properties {name=AngleFade, nodeColor=<rgba 0.5 0 1 1>}
  }
  MaterialExpressionNamedRerouteDeclaration_17 {
   class=MaterialExpressionNamedRerouteDeclaration
   inputs [None {from=MaterialExpressionQualitySwitch_0}]
   properties {name=DepthFade, nodeColor=<rgba 0 1 0.1797 1>}
  }
  MaterialExpressionNamedRerouteDeclaration_2 {
   class=MaterialExpressionNamedRerouteDeclaration
   inputs [None {from=MaterialExpressionMaterialFunctionCall_9, output=Result}]
   properties {name=Out Emissive, nodeColor=<rgba 0.6406 0 1 1>}
  }
  MaterialExpressionNamedRerouteDeclaration_3 {
   class=MaterialExpressionNamedRerouteDeclaration
   inputs [None {from=MaterialExpressionVertexColor_0, output=R}]
   properties {name=VC_R_Random, nodeColor=<rgba 1 0 0 1>}
  }
  MaterialExpressionNamedRerouteDeclaration_4 {
   class=MaterialExpressionNamedRerouteDeclaration
   inputs [None {from=MaterialExpressionVertexColor_0, output=G}]
   properties {name=VC_G_Random, nodeColor=<rgba 0 1 0 1>}
  }
  MaterialExpressionNamedRerouteDeclaration_5 {
   class=MaterialExpressionNamedRerouteDeclaration
   inputs [None {from=MaterialExpressionVertexColor_0, output=B}]
   properties {name=VC_B_Random, nodeColor=<rgba 0 0 1 1>}
  }
  MaterialExpressionNamedRerouteDeclaration_6 {
   class=MaterialExpressionNamedRerouteDeclaration
   inputs [None {from=MaterialExpressionStaticSwitch_1}]
   properties {name=Sprites_Color, nodeColor=<rgba 0 0.9297 1 1>}
  }
  MaterialExpressionNamedRerouteDeclaration_7 {
   class=MaterialExpressionNamedRerouteDeclaration
   inputs [None {from=MaterialExpressionStaticSwitchParameter_0}]
   properties {name=Sprites_Size, nodeColor=<rgba 0 0.9297 1 1>}
  }
  MaterialExpressionNamedRerouteDeclaration_8 {
   class=MaterialExpressionNamedRerouteDeclaration
   inputs [None {from=MaterialExpressionMaterialFunctionCall_2, output=Result}]
   properties {name=SpritesINDEX, nodeColor=<rgba 1 0 0.75 1>}
  }
  MaterialExpressionNamedRerouteDeclaration_9 {
   class=MaterialExpressionNamedRerouteDeclaration
   inputs [None {from=MaterialExpressionScalarParameter_10}]
   properties {name=SpritesCount, nodeColor=<rgba 0 0.6719 1 1>}
  }
  MaterialExpressionNamedRerouteUsage_0 {class=MaterialExpressionNamedRerouteUsage}
  MaterialExpressionNamedRerouteUsage_1 {class=MaterialExpressionNamedRerouteUsage}
  MaterialExpressionNamedRerouteUsage_10 {class=MaterialExpressionNamedRerouteUsage}
  MaterialExpressionNamedRerouteUsage_11 {class=MaterialExpressionNamedRerouteUsage}
  MaterialExpressionNamedRerouteUsage_12 {class=MaterialExpressionNamedRerouteUsage}
  MaterialExpressionNamedRerouteUsage_13 {class=MaterialExpressionNamedRerouteUsage}
  MaterialExpressionNamedRerouteUsage_14 {class=MaterialExpressionNamedRerouteUsage}
  MaterialExpressionNamedRerouteUsage_15 {class=MaterialExpressionNamedRerouteUsage}
  MaterialExpressionNamedRerouteUsage_16 {class=MaterialExpressionNamedRerouteUsage}
  MaterialExpressionNamedRerouteUsage_17 {class=MaterialExpressionNamedRerouteUsage}
  MaterialExpressionNamedRerouteUsage_18 {class=MaterialExpressionNamedRerouteUsage}
  MaterialExpressionNamedRerouteUsage_19 {class=MaterialExpressionNamedRerouteUsage}
  MaterialExpressionNamedRerouteUsage_2 {class=MaterialExpressionNamedRerouteUsage}
  MaterialExpressionNamedRerouteUsage_20 {class=MaterialExpressionNamedRerouteUsage}
  MaterialExpressionNamedRerouteUsage_21 {class=MaterialExpressionNamedRerouteUsage}
  MaterialExpressionNamedRerouteUsage_22 {class=MaterialExpressionNamedRerouteUsage}
  MaterialExpressionNamedRerouteUsage_3 {class=MaterialExpressionNamedRerouteUsage}
  MaterialExpressionNamedRerouteUsage_34 {class=MaterialExpressionNamedRerouteUsage}
  MaterialExpressionNamedRerouteUsage_4 {class=MaterialExpressionNamedRerouteUsage}
  MaterialExpressionNamedRerouteUsage_5 {class=MaterialExpressionNamedRerouteUsage}
  MaterialExpressionNamedRerouteUsage_6 {class=MaterialExpressionNamedRerouteUsage}
  MaterialExpressionNamedRerouteUsage_7 {class=MaterialExpressionNamedRerouteUsage}
  MaterialExpressionNamedRerouteUsage_8 {class=MaterialExpressionNamedRerouteUsage}
  MaterialExpressionNamedRerouteUsage_9 {class=MaterialExpressionNamedRerouteUsage}
  MaterialExpressionObjectRadius_0 {class=MaterialExpressionObjectRadius, inputs [], properties {}}
  MaterialExpressionPower_0 {
   class=MaterialExpressionPower
   inputs [
    Base {from=MaterialExpressionSphereMask_0}
    Exp {from=MaterialExpressionClamp_0}
   ]
   properties {constExponent=3}
  }
  MaterialExpressionQualitySwitch_0 {
   class=MaterialExpressionQualitySwitch
   inputs [
    Default {from=MaterialExpressionConstant_7, output=""}
    Low {from=MaterialExpressionConstant_7, output=""}
    High {from=MaterialExpressionDepthFade_1, output=""}
    Medium {from=MaterialExpressionConstant_7, output=""}
    Epic {from=MaterialExpressionDepthFade_1, output=""}
   ]
   properties {}
  }
  MaterialExpressionRayTracingQualitySwitch_0 {
   class=MaterialExpressionRayTracingQualitySwitch
   inputs [
    Normal {from=MaterialExpressionStaticSwitchParameter_4, output=""}
    RayTraced {from=MaterialExpressionConstant_0, output=""}
   ]
   properties {}
  }
  MaterialExpressionReroute_1 {class=MaterialExpressionReroute, inputs [None {from=MaterialExpressionMultiply_20}]}
  MaterialExpressionRound_0 {class=MaterialExpressionRound, inputs [None {from=MaterialExpressionNamedRerouteUsage_15, output=""}], properties {}}
  MaterialExpressionRound_3 {class=MaterialExpressionRound, inputs [None {from=MaterialExpressionNamedRerouteUsage_16, output=""}], properties {}}
  MaterialExpressionScalarParameter_0 {
   class=MaterialExpressionScalarParameter
   properties {defaultValue=1, group=GlowSprites 06 WPO Adjustment, parameterName=Camera Offset Amount}
  }
  MaterialExpressionScalarParameter_1 {class=MaterialExpressionScalarParameter, properties {group=GlowSprites 07 Randomness, parameterName=RandomSize_Min}}
  MaterialExpressionScalarParameter_10 {class=MaterialExpressionScalarParameter, properties {defaultValue=1, group=GlowSprites, parameterName=SpritesCount}}
  MaterialExpressionScalarParameter_11 {class=MaterialExpressionScalarParameter, properties {group=GlowSprites 01 SpitesGraph, parameterName=Hardness, sortPriority=11}}
  MaterialExpressionScalarParameter_12 {
   class=MaterialExpressionScalarParameter
   properties {defaultValue=0.5, group=GlowSprites 01 SpitesGraph, parameterName=Radius, sortPriority=10}
  }
  MaterialExpressionScalarParameter_13 {class=MaterialExpressionScalarParameter, properties {defaultValue=3, group=GlowSprites 01 SpitesGraph, parameterName=Power, sortPriority=13}}
  MaterialExpressionScalarParameter_14 {
   class=MaterialExpressionScalarParameter
   properties {defaultValue=100, group=GlowSprites 02 Opacity, parameterName=NearCameraFadeDist, sortPriority=20}
  }
  MaterialExpressionScalarParameter_15 {
   class=MaterialExpressionScalarParameter
   properties {defaultValue=1, group=GlowSprites 03 TOD Factor, parameterName=Value_Morning, sortPriority=40}
  }
  MaterialExpressionScalarParameter_16 {
   class=MaterialExpressionScalarParameter
   properties {defaultValue=1, group=GlowSprites 03 TOD Factor, parameterName=Value_Day, sortPriority=40}
  }
  MaterialExpressionScalarParameter_17 {
   class=MaterialExpressionScalarParameter
   properties {defaultValue=1, group=GlowSprites 03 TOD Factor, parameterName=Value_Evening, sortPriority=40}
  }
  MaterialExpressionScalarParameter_18 {
   class=MaterialExpressionScalarParameter
   properties {defaultValue=1, group=GlowSprites 03 TOD Factor, parameterName=Value_Night, sortPriority=40}
  }
  MaterialExpressionScalarParameter_19 {class=MaterialExpressionScalarParameter, properties {group=GlowSprites 02 Opacity, parameterName=Switch_CameraDepthFade, sortPriority=30}}
  MaterialExpressionScalarParameter_2 {class=MaterialExpressionScalarParameter, properties {defaultValue=1, group=GlowSprites 05 Size, parameterName=SpritesSize}}
  MaterialExpressionScalarParameter_20 {
   class=MaterialExpressionScalarParameter
   properties {defaultValue=1, group=GlowSprites 06 WPO Adjustment, parameterName=useActorRadius, sliderMax=1, sortPriority=0}
  }
  MaterialExpressionScalarParameter_21 {class=MaterialExpressionScalarParameter, properties {defaultValue=8, group=GlowSprites 02 Opacity, parameterName=AngleFade, sortPriority=20}}
  MaterialExpressionScalarParameter_23 {
   class=MaterialExpressionScalarParameter
   properties {defaultValue=100, group=GlowSprites 02 Opacity, parameterName=DepthFadeDist, sortPriority=22}
  }
  MaterialExpressionScalarParameter_24 {
   class=MaterialExpressionScalarParameter
   properties {defaultValue=1, group=GlowSprites 02 Opacity, parameterName=DepthFadeDistMulti, sortPriority=21}
  }
  MaterialExpressionScalarParameter_26 {
   class=MaterialExpressionScalarParameter
   properties {defaultValue=1, group=GlowSprites 02 Opacity, parameterName=EmissiveAlpha, sortPriority=1}
  }
  MaterialExpressionScalarParameter_3 {class=MaterialExpressionScalarParameter, properties {group=GlowSprites 07 Randomness, parameterName=RandomSize_Max}}
  MaterialExpressionScalarParameter_4 {
   class=MaterialExpressionScalarParameter
   properties {defaultValue=97.3942, group=GlowSprites 06 WPO Adjustment, parameterName=ActorRadius, sortPriority=1}
  }
  MaterialExpressionScalarParameter_56 {
   class=MaterialExpressionScalarParameter
   properties {defaultValue=100, group=GlowSprites 02 Opacity, parameterName=CameraDepthFade_Length, sortPriority=31}
  }
  MaterialExpressionScalarParameter_57 {class=MaterialExpressionScalarParameter, properties {defaultValue=110, group=GlowSprites 02 Opacity, parameterName=CameraDepthFade_Offset}}
  MaterialExpressionSetMaterialAttributes_0 {
   class=MaterialExpressionSetMaterialAttributes
   inputs [
    MaterialAttributes {from="", output=""}
    EmissiveColor {from=MaterialExpressionNamedRerouteUsage_9, output=""}
    Opacity {from=MaterialExpressionNamedRerouteUsage_10, output=""}
    WorldPositionOffset {from=MaterialExpressionNamedRerouteUsage_11, output=""}
   ]
   properties {attributeSetTypes [#1, #2, #3]}
  }
  MaterialExpressionSphereMask_0 {
   class=MaterialExpressionSphereMask
   inputs [
    A {from=MaterialExpressionNamedRerouteUsage_3, output=""}
    B {from=MaterialExpressionConstant_5, output=""}
    Radius {from=MaterialExpressionScalarParameter_12, output=""}
    Hardness {from=MaterialExpressionScalarParameter_11, output=""}
   ]
   properties {hardnessPercent=0}
  }
  MaterialExpressionStaticBool_0 {class=MaterialExpressionStaticBool}
  MaterialExpressionStaticSwitchParameter_0 {
   class=MaterialExpressionStaticSwitchParameter
   inputs [
    True {from=MaterialExpressionSwitch_2, output=""}
    False {from=MaterialExpressionScalarParameter_2, output=""}
   ]
   properties {group=GlowSprites 05 Size, parameterName=Use.SizeArray, sortPriority=0}
  }
  MaterialExpressionStaticSwitchParameter_2 {
   class=MaterialExpressionStaticSwitchParameter
   inputs [
    True {from=MaterialExpressionTextureSampleParameter2D_1, output=R}
    False {from=MaterialExpressionNamedRerouteUsage_21, output=""}
   ]
   properties {group=GlowSprites 01 SpitesGraph, parameterName=Use.MaskAsGraph, sortPriority=0}
  }
  MaterialExpressionStaticSwitchParameter_3 {
   class=MaterialExpressionStaticSwitchParameter
   inputs [
    True {from=MaterialExpressionSwitch_0, output=""}
    False {from=MaterialExpressionVectorParameter_0, output=RGB}
   ]
   properties {group=GlowSprites 04 Color, parameterName=Use.ColorArray, sortPriority=0}
  }
  MaterialExpressionStaticSwitchParameter_4 {
   class=MaterialExpressionStaticSwitchParameter
   inputs [
    True {from=MaterialExpressionNamedRerouteUsage_0, output=""}
    False {from=MaterialExpressionLinearInterpolate_19, output=""}
   ]
   properties {group=GlowSprites, parameterName=Use.DebugDisplaySpritesID, sortPriority=0}
  }
  MaterialExpressionStaticSwitchParameter_5 {
   class=MaterialExpressionStaticSwitchParameter
   inputs [
    True {from=MaterialExpressionConstant_2, output=""}
    False {from=MaterialExpressionAdd_0, output=""}
   ]
   properties {group=GlowSprites 02 Opacity, parameterName="Use AngleFade (Require fixed size/rotation setup for sprites)", sortPriority=0}
  }
  MaterialExpressionStaticSwitchParameter_6 {
   class=MaterialExpressionStaticSwitchParameter
   inputs [
    True {from=MaterialExpressionMaterialFunctionCall_7, output=Result}
    False {from=MaterialExpressionConstant_8, output=""}
   ]
   properties {group=GlowSprites 02 Opacity, parameterName="Use AngleFade (Require fixed size/rotation setup for sprites)", sortPriority=0}
  }
  MaterialExpressionStaticSwitchParameter_7 {
   class=MaterialExpressionStaticSwitchParameter
   inputs [
    True {from=MaterialExpressionTextureSampleParameter2D_0, output=RGB}
    False {from=MaterialExpressionStaticSwitchParameter_3, output=""}
   ]
   properties {defaultValue=true, group=GlowSprites 04 Color, parameterName=Use.TextureAsColor, sortPriority=0}
  }
  MaterialExpressionStaticSwitch_0 {
   class=MaterialExpressionStaticSwitch
   inputs [
    True {from=MaterialExpressionFunctionInput_0, output=""}
    False {from=MaterialExpressionStaticSwitchParameter_2, output=""}
    Value {from=MaterialExpressionNamedRerouteUsage_6, output=""}
   ]
   properties {}
  }
  MaterialExpressionStaticSwitch_1 {
   class=MaterialExpressionStaticSwitch
   inputs [
    True {from=MaterialExpressionFunctionInput_1, output=""}
    False {from=MaterialExpressionStaticSwitchParameter_7, output=""}
    Value {from=MaterialExpressionNamedRerouteUsage_18, output=""}
   ]
   properties {}
  }
  MaterialExpressionSubtract_0 {
   class=MaterialExpressionSubtract
   inputs [
    A {from=MaterialExpressionMultiply_3}
    B {from=""}
   ]
   properties {constB=0.5}
  }
  MaterialExpressionSubtract_4 {
   class=MaterialExpressionSubtract
   inputs [
    A {from=MaterialExpressionRound_0}
    B {from=""}
   ]
  }
  MaterialExpressionSubtract_5 {
   class=MaterialExpressionSubtract
   inputs [
    A {from=MaterialExpressionNamedRerouteUsage_4}
    B {from=""}
   ]
   properties {constB=0.5}
  }
  MaterialExpressionSwitch_0 {
   class=MaterialExpressionSwitch
   inputs [
    SwitchValue {from=MaterialExpressionSubtract_4, output=""}
    Default {from="", output=""}
    1 {from=MaterialExpressionVectorParameter_1, output=RGB}
    2 {from=MaterialExpressionVectorParameter_2, output=RGB}
    3 {from=MaterialExpressionVectorParameter_3, output=RGB}
    4 {from=MaterialExpressionVectorParameter_4, output=RGB}
    5 {from=MaterialExpressionVectorParameter_5, output=RGB}
    6 {from=MaterialExpressionVectorParameter_6, output=RGB}
    7 {from=MaterialExpressionVectorParameter_7, output=RGB}
    8 {from=MaterialExpressionVectorParameter_8, output=RGB}
    9 {from=MaterialExpressionVectorParameter_15, output=RGB}
    10 {from=MaterialExpressionVectorParameter_16, output=RGB}
    11 {from=MaterialExpressionVectorParameter_17, output=RGB}
    12 {from=MaterialExpressionVectorParameter_18, output=RGB}
    13 {from=MaterialExpressionVectorParameter_19, output=RGB}
    14 {from=MaterialExpressionVectorParameter_20, output=RGB}
    15 {from=MaterialExpressionVectorParameter_21, output=RGB}
    16 {from=MaterialExpressionVectorParameter_22, output=RGB}
   ]
   properties {
    desc=Color Sets
    inputs [
     {
      inputName=1
      input {
       expression=&self:MaterialExpressionVectorParameter_1
       inputName=None
       mask=1
       maskA=0
       maskB=1
       maskG=1
       maskR=1
       outputIndex=0
      }
     }
     {
      inputName=2
      input {
       expression=&self:MaterialExpressionVectorParameter_2
       inputName=None
       mask=1
       maskA=0
       maskB=1
       maskG=1
       maskR=1
       outputIndex=0
      }
     }
     {
      inputName=3
      input {
       expression=&self:MaterialExpressionVectorParameter_3
       inputName=None
       mask=1
       maskA=0
       maskB=1
       maskG=1
       maskR=1
       outputIndex=0
      }
     }
     {
      inputName=4
      input {
       expression=&self:MaterialExpressionVectorParameter_4
       inputName=None
       mask=1
       maskA=0
       maskB=1
       maskG=1
       maskR=1
       outputIndex=0
      }
     }
     {
      inputName=5
      input {
       expression=&self:MaterialExpressionVectorParameter_5
       inputName=None
       mask=1
       maskA=0
       maskB=1
       maskG=1
       maskR=1
       outputIndex=0
      }
     }
     {
      inputName=6
      input {
       expression=&self:MaterialExpressionVectorParameter_6
       inputName=None
       mask=1
       maskA=0
       maskB=1
       maskG=1
       maskR=1
       outputIndex=0
      }
     }
     {
      inputName=7
      input {
       expression=&self:MaterialExpressionVectorParameter_7
       inputName=None
       mask=1
       maskA=0
       maskB=1
       maskG=1
       maskR=1
       outputIndex=0
      }
     }
     {
      inputName=8
      input {
       expression=&self:MaterialExpressionVectorParameter_8
       inputName=None
       mask=1
       maskA=0
       maskB=1
       maskG=1
       maskR=1
       outputIndex=0
      }
     }
     {
      inputName=9
      input {
       expression=&self:MaterialExpressionVectorParameter_15
       inputName=None
       mask=1
       maskA=0
       maskB=1
       maskG=1
       maskR=1
       outputIndex=0
      }
     }
     {
      inputName=10
      input {
       expression=&self:MaterialExpressionVectorParameter_16
       inputName=None
       mask=1
       maskA=0
       maskB=1
       maskG=1
       maskR=1
       outputIndex=0
      }
     }
     {
      inputName=11
      input {
       expression=&self:MaterialExpressionVectorParameter_17
       inputName=None
       mask=1
       maskA=0
       maskB=1
       maskG=1
       maskR=1
       outputIndex=0
      }
     }
     {
      inputName=12
      input {
       expression=&self:MaterialExpressionVectorParameter_18
       inputName=None
       mask=1
       maskA=0
       maskB=1
       maskG=1
       maskR=1
       outputIndex=0
      }
     }
     {
      inputName=13
      input {
       expression=&self:MaterialExpressionVectorParameter_19
       inputName=None
       mask=1
       maskA=0
       maskB=1
       maskG=1
       maskR=1
       outputIndex=0
      }
     }
     {
      inputName=14
      input {
       expression=&self:MaterialExpressionVectorParameter_20
       inputName=None
       mask=1
       maskA=0
       maskB=1
       maskG=1
       maskR=1
       outputIndex=0
      }
     }
     {
      inputName=15
      input {
       expression=&self:MaterialExpressionVectorParameter_21
       inputName=None
       mask=1
       maskA=0
       maskB=1
       maskG=1
       maskR=1
       outputIndex=0
      }
     }
     {
      inputName=16
      input {
       expression=&self:MaterialExpressionVectorParameter_22
       inputName=None
       mask=1
       maskA=0
       maskB=1
       maskG=1
       maskR=1
       outputIndex=0
      }
     }
    ]
   }
  }
  MaterialExpressionSwitch_2 {
   class=MaterialExpressionSwitch
   inputs [
    SwitchValue {from=MaterialExpressionSubtract_4, output=""}
    Default {from="", output=""}
    1 {from=MaterialExpressionVectorParameter_11, output=ID01}
    2 {from=MaterialExpressionVectorParameter_11, output=ID01}
    3 {from=MaterialExpressionVectorParameter_11, output=ID01}
    4 {from=MaterialExpressionVectorParameter_11, output=ID01}
    5 {from=MaterialExpressionVectorParameter_12, output=ID05}
    6 {from=MaterialExpressionVectorParameter_12, output=ID05}
    7 {from=MaterialExpressionVectorParameter_12, output=ID05}
    8 {from=MaterialExpressionVectorParameter_12, output=ID05}
    9 {from=MaterialExpressionVectorParameter_13, output=ID09}
    10 {from=MaterialExpressionVectorParameter_13, output=ID09}
    11 {from=MaterialExpressionVectorParameter_13, output=ID09}
    12 {from=MaterialExpressionVectorParameter_13, output=ID09}
    13 {from=MaterialExpressionVectorParameter_14, output=ID13}
    14 {from=MaterialExpressionVectorParameter_14, output=ID13}
    15 {from=MaterialExpressionVectorParameter_14, output=ID13}
    16 {from=MaterialExpressionVectorParameter_14, output=ID13}
   ]
   properties {
    desc=Color Sets
    inputs [
     {
      inputName=1
      input {
       expression=&self:MaterialExpressionVectorParameter_11
       inputName=None
       mask=1
       maskA=0
       maskB=0
       maskG=0
       maskR=1
       outputIndex=1
      }
     }
     {
      inputName=2
      input {
       expression=&self:MaterialExpressionVectorParameter_11
       inputName=None
       mask=1
       maskA=0
       maskB=0
       maskG=1
       maskR=0
       outputIndex=2
      }
     }
     {
      inputName=3
      input {
       expression=&self:MaterialExpressionVectorParameter_11
       inputName=None
       mask=1
       maskA=0
       maskB=1
       maskG=0
       maskR=0
       outputIndex=3
      }
     }
     {
      inputName=4
      input {
       expression=&self:MaterialExpressionVectorParameter_11
       inputName=None
       mask=1
       maskA=1
       maskB=0
       maskG=0
       maskR=0
       outputIndex=4
      }
     }
     {
      inputName=5
      input {
       expression=&self:MaterialExpressionVectorParameter_12
       inputName=None
       mask=1
       maskA=0
       maskB=0
       maskG=0
       maskR=1
       outputIndex=1
      }
     }
     {
      inputName=6
      input {
       expression=&self:MaterialExpressionVectorParameter_12
       inputName=None
       mask=1
       maskA=0
       maskB=0
       maskG=1
       maskR=0
       outputIndex=2
      }
     }
     {
      inputName=7
      input {
       expression=&self:MaterialExpressionVectorParameter_12
       inputName=None
       mask=1
       maskA=0
       maskB=1
       maskG=0
       maskR=0
       outputIndex=3
      }
     }
     {
      inputName=8
      input {
       expression=&self:MaterialExpressionVectorParameter_12
       inputName=None
       mask=1
       maskA=1
       maskB=0
       maskG=0
       maskR=0
       outputIndex=4
      }
     }
     {
      inputName=9
      input {
       expression=&self:MaterialExpressionVectorParameter_13
       inputName=None
       mask=1
       maskA=0
       maskB=0
       maskG=0
       maskR=1
       outputIndex=1
      }
     }
     {
      inputName=10
      input {
       expression=&self:MaterialExpressionVectorParameter_13
       inputName=None
       mask=1
       maskA=0
       maskB=0
       maskG=1
       maskR=0
       outputIndex=2
      }
     }
     {
      inputName=11
      input {
       expression=&self:MaterialExpressionVectorParameter_13
       inputName=None
       mask=1
       maskA=0
       maskB=1
       maskG=0
       maskR=0
       outputIndex=3
      }
     }
     {
      inputName=12
      input {
       expression=&self:MaterialExpressionVectorParameter_13
       inputName=None
       mask=1
       maskA=1
       maskB=0
       maskG=0
       maskR=0
       outputIndex=4
      }
     }
     {
      inputName=13
      input {
       expression=&self:MaterialExpressionVectorParameter_14
       inputName=None
       mask=1
       maskA=0
       maskB=0
       maskG=0
       maskR=1
       outputIndex=1
      }
     }
     {
      inputName=14
      input {
       expression=&self:MaterialExpressionVectorParameter_14
       inputName=None
       mask=1
       maskA=0
       maskB=0
       maskG=1
       maskR=0
       outputIndex=2
      }
     }
     {
      inputName=15
      input {
       expression=&self:MaterialExpressionVectorParameter_14
       inputName=None
       mask=1
       maskA=0
       maskB=1
       maskG=0
       maskR=0
       outputIndex=3
      }
     }
     {
      inputName=16
      input {
       expression=&self:MaterialExpressionVectorParameter_14
       inputName=None
       mask=1
       maskA=1
       maskB=0
       maskG=0
       maskR=0
       outputIndex=4
      }
     }
    ]
   }
  }
  MaterialExpressionTextureCoordinate_0 {class=MaterialExpressionTextureCoordinate, properties {coordinateIndex=1}}
  MaterialExpressionTextureSampleParameter2D_0 {
   class=MaterialExpressionTextureSampleParameter2D
   inputs [
    UVs {from="", output=""}
    Apply View MipBias {from="", output=""}
   ]
   properties {
    group=GlowSprites 04 Color
    parameterName=SpritesColorMap
    samplerSource=SSM_Clamp_WorldGroupSettings
    samplerType=SAMPLERTYPE_Masks
    texture=/Game/Global/Textures/Default/Blanks/T_White
   }
  }
  MaterialExpressionTextureSampleParameter2D_1 {
   class=MaterialExpressionTextureSampleParameter2D
   inputs [
    UVs {from=MaterialExpressionNamedRerouteUsage_2, output=""}
    Apply View MipBias {from="", output=""}
   ]
   properties {
    group=GlowSprites 01 SpitesGraph
    parameterName=GlowSpritesMask
    samplerSource=SSM_Wrap_WorldGroupSettings
    sortPriority=2
    texture=/Game/Global/Materials/Character_Parent_2023/GlowSprites_Parent_2026/Textures/T_GlowSprite_Sphere
   }
  }
  MaterialExpressionTransform_10 {
   class=MaterialExpressionTransform
   inputs [None {from=MaterialExpressionConstant3Vector_22}]
   properties {transformSourceType=TRANSFORMSOURCE_Camera}
  }
  MaterialExpressionTransform_9 {
   class=MaterialExpressionTransform
   inputs [None {from=MaterialExpressionConstant3Vector_23}]
   properties {transformSourceType=TRANSFORMSOURCE_Camera}
  }
  MaterialExpressionVectorParameter_0 {class=MaterialExpressionVectorParameter, properties {group=GlowSprites 04 Color, parameterName=SpritesColor, defaultValue=<rgba 1 1 1 0>}}
  MaterialExpressionVectorParameter_1 {class=MaterialExpressionVectorParameter, properties {group=GlowSprites 04 ColorArry, parameterName=ID_01_Color, defaultValue=<rgba 1 0 0 1>}}
  MaterialExpressionVectorParameter_11 {
   class=MaterialExpressionVectorParameter
   properties {group=GlowSprites 05 SizeArray, parameterName=ID_01-04_Size, channelNames {a=ID04, b=ID03, g=ID02, r=ID01}, defaultValue=<rgba 1 1 1 1>}
  }
  MaterialExpressionVectorParameter_12 {
   class=MaterialExpressionVectorParameter
   properties {group=GlowSprites 05 SizeArray, parameterName=ID_05-08_Size, channelNames {a=ID08, b=ID07, g=ID06, r=ID05}, defaultValue=<rgba 1 1 1 1>}
  }
  MaterialExpressionVectorParameter_13 {
   class=MaterialExpressionVectorParameter
   properties {group=GlowSprites 05 SizeArray, parameterName=ID_08-12_Size, channelNames {a=ID12, b=ID11, g=ID10, r=ID09}, defaultValue=<rgba 1 1 1 1>}
  }
  MaterialExpressionVectorParameter_14 {
   class=MaterialExpressionVectorParameter
   properties {group=GlowSprites 05 SizeArray, parameterName=ID_13-16_Size, channelNames {a=ID16, b=ID15, g=ID14, r=ID13}, defaultValue=<rgba 1 1 1 1>}
  }
  MaterialExpressionVectorParameter_15 {class=MaterialExpressionVectorParameter, properties {group=GlowSprites 04 ColorArry, parameterName=ID_09_Color, defaultValue=<rgba 1 0 0 1>}}
  MaterialExpressionVectorParameter_16 {class=MaterialExpressionVectorParameter, properties {group=GlowSprites 04 ColorArry, parameterName=ID_10_Color, defaultValue=<rgba 0 1 0 1>}}
  MaterialExpressionVectorParameter_17 {class=MaterialExpressionVectorParameter, properties {group=GlowSprites 04 ColorArry, parameterName=ID_11_Color, defaultValue=<rgba 0 0 1 1>}}
  MaterialExpressionVectorParameter_18 {class=MaterialExpressionVectorParameter, properties {group=GlowSprites 04 ColorArry, parameterName=ID_12_Color, defaultValue=<rgba 1 0 1 1>}}
  MaterialExpressionVectorParameter_19 {class=MaterialExpressionVectorParameter, properties {group=GlowSprites 04 ColorArry, parameterName=ID_13_Color, defaultValue=<rgba 1 1 0 1>}}
  MaterialExpressionVectorParameter_2 {class=MaterialExpressionVectorParameter, properties {group=GlowSprites 04 ColorArry, parameterName=ID_02_Color, defaultValue=<rgba 0 1 0 1>}}
  MaterialExpressionVectorParameter_20 {class=MaterialExpressionVectorParameter, properties {group=GlowSprites 04 ColorArry, parameterName=ID_14_Color, defaultValue=<rgba 0 1 1 1>}}
  MaterialExpressionVectorParameter_21 {
   class=MaterialExpressionVectorParameter
   properties {group=GlowSprites 04 ColorArry, parameterName=ID_15_Color, defaultValue=<rgba 1 0.5 0 1>}
  }
  MaterialExpressionVectorParameter_22 {class=MaterialExpressionVectorParameter, properties {group=GlowSprites 04 ColorArry, parameterName=ID_16_Color, defaultValue=<rgba 1 1 1 1>}}
  MaterialExpressionVectorParameter_3 {class=MaterialExpressionVectorParameter, properties {group=GlowSprites 04 ColorArry, parameterName=ID_03_Color, defaultValue=<rgba 0 0 1 1>}}
  MaterialExpressionVectorParameter_4 {class=MaterialExpressionVectorParameter, properties {group=GlowSprites 04 ColorArry, parameterName=ID_04_Color, defaultValue=<rgba 1 0 1 1>}}
  MaterialExpressionVectorParameter_5 {class=MaterialExpressionVectorParameter, properties {group=GlowSprites 04 ColorArry, parameterName=ID_05_Color, defaultValue=<rgba 1 1 0 1>}}
  MaterialExpressionVectorParameter_6 {class=MaterialExpressionVectorParameter, properties {group=GlowSprites 04 ColorArry, parameterName=ID_06_Color, defaultValue=<rgba 0 1 1 1>}}
  MaterialExpressionVectorParameter_7 {
   class=MaterialExpressionVectorParameter
   properties {group=GlowSprites 04 ColorArry, parameterName=ID_07_Color, defaultValue=<rgba 1 0.5 0 1>}
  }
  MaterialExpressionVectorParameter_8 {class=MaterialExpressionVectorParameter, properties {group=GlowSprites 04 ColorArry, parameterName=ID_08_Color, defaultValue=<rgba 1 1 1 1>}}
  MaterialExpressionVertexColor_0 {class=MaterialExpressionVertexColor, inputs [], properties {}}
 }}
