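# Translucent unlit parent material for glow sprites on skeletal meshes, built on material attributes.
# MF_GlowSprites produces the look and exposes optional custom glow graph inputs, alongside a tech-art overrides function.
# An FP_Mode static switch, on by default, passes the result through first-person mesh camera-offset and FOV-interpolation functions, or outputs it unmodified when off.
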Material /Game/Global/Materials/Character_Parent_2023/GlowSprites_Parent_2026/M_GlowSprites_PARENT_2026
graph {
 properties {
  bAutomaticallySetUsageInEditor=false
  bCastRayTracedShadows=false
  bUseMaterialAttributes=true
  bUsedWithSkeletalMesh=true
  blendMode=BLEND_Translucent
  shadingModel=MSM_Unlit
  usedShadingModels=Unlit
 }
 nodes [
  MaterialFunctionCall() {materialFunction=&MF_TechArtOverrides}
  n1 = MaterialFunctionCall!(UseCustomGlowGraph, CustomGlowGraph, CustomGlowColor) {materialFunction=&/MF_GlowSprites}
  StaticSwitchParameter(
   True=MaterialFunctionCall!(In=MaterialFunctionCall!(In=@n1.MA) {materialFunction=/Game/Packages/Fortress_SharedMaterials/Base_Material_Functions/BaseCharacterShaderFunctions/MF_FPMeshCameraOffset}.Result) {materialFunction=/Game/Packages/Fortress_SharedMaterials/Base_Material_Functions/BaseCharacterShaderFunctions/MF_InterpFP_FOV}.Result
   False=@n1.MA
  ) {defaultValue=true, group=FP Mode, parameterName=FP_Mode, sortPriority=10}
 ]
}
